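# UI material for the tech-art texture tools that inspects the alpha channel of a Texture parameter at a chosen mip level.
# A MipBias parameter biases the texture sample.
# Two chained If nodes pass alpha through where it is 0 or 1, showing black or white, and flag every other value in hot pink, exposing non-binary alpha.
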Material /Game/Blueprints/Blutilities/ForTechArt/AssetTools/TextureTools/Materials/TextureAlphaInspectionMIPPED_UIMat
graph {
 properties {materialDomain=MD_UI, pixelDepthOffsetMode=PDOM_Legacy, usedShadingModels=Default Lit}
 nodes [
  n1 = TextureSampleParameter2D!(UVs, Bias=ScalarParameter() {parameterName=MipBias}, Apply View MipBias) {mipValueMode=TMVM_MipBias, parameterName=Texture, texture=/Engine/ArtTools/RenderToTexture/Textures/T_EV_BlankWhite_01}
  n2 = Constant3Vector() {constant=<rgba 1 0 0.8611 0>}
  n3 = If(A=@n1.A, B=Constant() {r=1}, "A > B"=@n2, "A == B"=@n1.A, "A < B"=@n2)
  If(A=@n1.A, B=Constant(), "A > B"=@n3, "A == B"=@n1.A, "A < B"=@n3)
 ]
}
