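# UI material for tech-art texture tools that visualizes a texture's alpha channel in three bands.
# A Texture parameter's alpha feeds two nested If nodes that compare it against one and zero.
# Fully opaque texels keep white, fully transparent ones keep black, and any partial alpha becomes flat mid-gray, exposing non-binary alpha.
# A magenta color and two scalar constants sit unconnected.
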Material /Game/Blueprints/Blutilities/ForTechArt/AssetTools/TextureTools/Materials/TextureAlphaInspection_UIMat
graph {
 properties {materialDomain=MD_UI, pixelDepthOffsetMode=PDOM_Legacy, usedShadingModels=Default Lit}
 nodes [
  Constant3Vector() {constant=<rgba 1 0 0.8888 0>}
  Constant()
  Constant()
  n1 = TextureSampleParameter2D() {parameterName=Texture, texture=/Game/Athena/UI/PostMatch/AthenaChallenge/Textures/T_UI_CheckMark}
  If(
   A=@n1.A
   B=Constant() {r=1}
   "A > B"=@n1.A
   "A == B"=@n1.A
   "A < B"=If(A=@n1.A, B=Constant(), "A > B"=Constant3Vector() {constant=<rgba 0.5 0.5 0.5 0>}, "A == B"=@n1.A, "A < B"=@n1.A)
  )
 ]
}
