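# UI-domain material in the tech-art TextureTools editor utility folder, apparently for previewing a texture in a tool widget.
# Its only node is a 2D texture sample parameter named Texture, defaulting to an engine blank white texture, which material instances can override.
# The graph shows no other processing and no link from the sample to a material output.
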Material /Game/Blueprints/Blutilities/ForTechArt/AssetTools/TextureTools/Materials/TextureDisplay_UIMat
graph {
 properties {materialDomain=MD_UI, pixelDepthOffsetMode=PDOM_Legacy, usedShadingModels=Default Lit}
 nodes [TextureSampleParameter2D() {parameterName=Texture, texture=/Engine/ArtTools/RenderToTexture/Textures/T_EV_BlankWhite_01}]
}
